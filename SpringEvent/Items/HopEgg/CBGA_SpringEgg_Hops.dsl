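; Spring event Hop Egg gatherable item that either of two interaction types can use.
; A first interaction is allowed for any valid pawn, while a second interaction also requires the gather-enabled data row to be true.
; On a second interaction, the gather event calls its parent, then sets the bHasBeenCollected flag.
; That flag's RepNotify plays the egg pickup sound at the actor's location.
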(event Interaction|EventOnGatherOrInteract
  (params (InteractingPawn :type Fort%20Pawn%20Object%20Reference) (InteractionBeingAttempted :type EInteractionBeingAttempted%20Enum :default "FirstInteraction"))
  ;[1 1 1] Set Has Been Collected to trigger initial pickup sound
  (|CallParentFunction InteractingPawn InteractionBeingAttempted)
  (switch Utilities|FlowControl|Switch|SwitchonEInteractionBeingAttempted InteractionBeingAttempted
    (:SecondInteraction
      (Variables|Default|SetbHasBeenCollected true))))

(fn OnInteractAllowedToConsume
  (params (InteractingPawn :type Fort%20Pawn%20Object%20Reference) (InteractionBeingAttempted :type EInteractionBeingAttempted%20Enum :default "FirstInteraction"))
  (Utilities|IsValid InteractingPawn
    (:Is%20Valid
      ;[1 1 1] If second interact allow gather if enabled
      (switch Utilities|FlowControl|Switch|SwitchonEInteractionBeingAttempted InteractionBeingAttempted
        (:FirstInteraction
          (return true))
        (:SecondInteraction
          (return (Class|AthenaFunctionLibrary|RowtoBool (Variables|Gather|GetRow_GatherEnabled))))))
    (:Is%20Not%20Valid
      (return false))))

(fn OnRep_bHasBeenCollected
  (if (Variables|Default|GetbHasBeenCollected)
    (Audio|PlaySoundatLocation "/SpringEvent/Items/Audio/SpringEggs_Pickup_Cue.SpringEggs_Pickup_Cue" (Transformation|GetActorLocation self))))

(fn ConstructionScript
  (|CallParentFunction))
